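# Sound cue for the Stone Viper Female pickaxe hitting an enemy, layering randomized impact and vocal samples.
# A quality-level switch selects among richer and lighter variants.
# The first level mixes impact and vox, sometimes adding a delayed, quieter recoil vox with its own foley attenuation.
# The second and fourth share an impact with occasional lower-volume vox; the third plays only a briefly delayed impact.
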SoundCue /BRCosmetics/Sounds/PickAxes/StoneViperFemale/PickaxeImpactEnemy_StoneViperFemale_Cue
root=QualityLevel_0
nodes [
 {
  id=Attenuation_0
  children [Delay_0]
  properties {attenuationSettings=&Foley_Attenuation_Basic_Att}
 }
 {id=Delay_0, children [Modulator_5], properties {delayMax=0.6, delayMin=0.5}}
 {id=Delay_3, children [Modulator_3], properties {delayMax=0.05, delayMin=0.05}}
 {id=Mixer_0, children [Modulator_1,Modulator_0], properties {inputVolume [1,1]}}
 {id=Mixer_1, children [Modulator_2,Modulator_4], properties {inputVolume [1,0.6]}}
 {id=Mixer_2, children [Modulator_1, Modulator_0, Attenuation_0], properties {inputVolume [1,1,0.6]}}
 {id=Modulator_0, children [Random_4]}
 {id=Modulator_1, children [Random_1]}
 {id=Modulator_2, children [Random_2]}
 {id=Modulator_3, children [Random_6]}
 {id=Modulator_4, children [Random_0]}
 {id=Modulator_5, children [Random_3]}
 {id=QualityLevel_0, children [Random_5,Random_7,Delay_3,Random_7]}
 {id=Random_0, children [WavePlayer_0,WavePlayer_1,WavePlayer_2], properties {weights [1,1,1]}}
 {
  id=Random_1
  children [WavePlayer_6, WavePlayer_7, WavePlayer_8, WavePlayer_9, WavePlayer_10, WavePlayer_11]
  properties {weights [1,1,1,1,1,1]}
 }
 {id=Random_2, children [WavePlayer_12, WavePlayer_13, WavePlayer_14, WavePlayer_15], properties {weights [1,1,1,1]}}
 {
  id=Random_3
  children [WavePlayer_16, WavePlayer_17, WavePlayer_18, WavePlayer_19]
  properties {bRandomizeWithoutReplacement=false, weights [1,1,1,1]}
 }
 {
  id=Random_4
  children [WavePlayer_20, WavePlayer_21, WavePlayer_22, WavePlayer_23, WavePlayer_24, WavePlayer_25]
  properties {bRandomizeWithoutReplacement=false, weights [1,1,1,1,1,1]}
 }
 {id=Random_5, children [Mixer_0,Mixer_2], properties {bRandomizeWithoutReplacement=false, weights [1,0.5]}}
 {id=Random_6, children [WavePlayer_3, WavePlayer_4, WavePlayer_5, WavePlayer_26], properties {weights [1,1,1,1]}}
 {id=Random_7, children [Modulator_2,Mixer_1], properties {bRandomizeWithoutReplacement=false, weights [1,0.4]}}
 {id=WavePlayer_0, properties {soundWaveAssetPtr=&/Pickaxe_Stone_Viper_Female_Impact_Vox_01}}
 {id=WavePlayer_1, properties {soundWaveAssetPtr=&/Pickaxe_Stone_Viper_Female_Impact_Vox_02}}
 {id=WavePlayer_10, properties {soundWaveAssetPtr=&/Pickaxe_Stone_Viper_Female_Impact_04}}
 {id=WavePlayer_11, properties {soundWaveAssetPtr=&/Pickaxe_Stone_Viper_Female_Impact_05}}
 {id=WavePlayer_12, properties {soundWaveAssetPtr=&/Pickaxe_Stone_Viper_Female_Impact_01}}
 {id=WavePlayer_13, properties {soundWaveAssetPtr=&/Pickaxe_Stone_Viper_Female_Impact_03}}
 {id=WavePlayer_14, properties {soundWaveAssetPtr=&/Pickaxe_Stone_Viper_Female_Impact_02}}
 {id=WavePlayer_15, properties {soundWaveAssetPtr=&Pickaxe_Stone_Viper_Female_Impact_04}}
 {id=WavePlayer_16, properties {soundWaveAssetPtr=&/Pickaxe_StoneViperF_Impact_Vox_Recoil_01}}
 {id=WavePlayer_17, properties {soundWaveAssetPtr=&/Pickaxe_StoneViperF_Impact_Vox_Recoil_02}}
 {id=WavePlayer_18, properties {soundWaveAssetPtr=&/Pickaxe_StoneViperF_Impact_Vox_Recoil_03}}
 {id=WavePlayer_19, properties {soundWaveAssetPtr=&/Pickaxe_StoneViperF_Impact_Vox_Recoil_04}}
 {id=WavePlayer_2, properties {soundWaveAssetPtr=&/Pickaxe_Stone_Viper_Female_Impact_Vox_03}}
 {id=WavePlayer_20, properties {soundWaveAssetPtr=&Pickaxe_Stone_Viper_Female_Impact_Vox_01}}
 {id=WavePlayer_21, properties {soundWaveAssetPtr=&Pickaxe_Stone_Viper_Female_Impact_Vox_02}}
 {id=WavePlayer_22, properties {soundWaveAssetPtr=&Pickaxe_Stone_Viper_Female_Impact_Vox_03}}
 {id=WavePlayer_23, properties {soundWaveAssetPtr=&/Pickaxe_Stone_Viper_Female_Impact_Vox_04}}
 {id=WavePlayer_24, properties {soundWaveAssetPtr=&/Pickaxe_Stone_Viper_Female_Impact_Vox_05}}
 {id=WavePlayer_25, properties {soundWaveAssetPtr=&/Pickaxe_Stone_Viper_Female_Impact_Vox_06}}
 {id=WavePlayer_26, properties {soundWaveAssetPtr=&Pickaxe_Stone_Viper_Female_Impact_04}}
 {id=WavePlayer_3, properties {soundWaveAssetPtr=&Pickaxe_Stone_Viper_Female_Impact_01}}
 {id=WavePlayer_4, properties {soundWaveAssetPtr=&Pickaxe_Stone_Viper_Female_Impact_03}}
 {id=WavePlayer_5, properties {soundWaveAssetPtr=&Pickaxe_Stone_Viper_Female_Impact_02}}
 {id=WavePlayer_6, properties {soundWaveAssetPtr=&/Pickaxe_Stone_Viper_Female_Impact_06}}
 {id=WavePlayer_7, properties {soundWaveAssetPtr=&Pickaxe_Stone_Viper_Female_Impact_01}}
 {id=WavePlayer_8, properties {soundWaveAssetPtr=&Pickaxe_Stone_Viper_Female_Impact_02}}
 {id=WavePlayer_9, properties {soundWaveAssetPtr=&Pickaxe_Stone_Viper_Female_Impact_03}}
]
properties {
 attenuationSettings=/Game/Sounds/Attenuation/Fort_Pickaxe_Impact_Att
 audioPropertiesSheet=/BRCosmetics/Sounds/Settings/PropertySheets/Pickaxe/APS_Cosmetic_Pickaxe_Impact
 bPrimeOnLoad=true
 maxDistance=4300
 soundClassObject=&Fort_WeaponMelee_PSM
 volumeMultiplier=0.4
 concurrencySet [&Player_Melee_Weapon_Conc]
}
