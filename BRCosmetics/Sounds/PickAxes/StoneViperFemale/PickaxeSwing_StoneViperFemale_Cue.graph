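# Sound cue for the Stone Viper Female pickaxe swing, switched by quality level.
# The highest level mixes a close-attenuated layer of random swing and vocal waves with a medium-attenuated layer of swings and distant fire-axe swings.
# Lower levels use only a close layer from smaller wave pools, with vocals added on a weighted random chance.
# Modulators vary pitch, and overridden concurrency caps voices per owner, stopping the oldest.
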SoundCue /BRCosmetics/Sounds/PickAxes/StoneViperFemale/PickaxeSwing_StoneViperFemale_Cue
root=QualityLevel_1
nodes [
 {id=Attenuation_1, children [Modulator_0], properties {attenuationSettings=/Game/Sounds/Attenuation/Fort_Weapon_Pickaxe_Close_Att}}
 {id=Attenuation_2, children [Modulator_1], properties {attenuationSettings=&Fort_Weapon_Pickaxe_Close_Att}}
 {id=Attenuation_6, children [Modulator_2], properties {attenuationSettings=&Fort_Weapon_Pickaxe_Close_Att}}
 {id=Attenuation_7, children [Modulator_8], properties {attenuationSettings=/Game/Sounds/Attenuation/Fort_Weapon_Pickaxe_Medium_Att}}
 {id=Mixer_0, children [Random_2,Random_0], properties {inputVolume [1,0.5]}}
 {id=Mixer_1, children [Random_7,Random_15], properties {inputVolume [1,0.6]}}
 {id=Mixer_5, children [Random_18,Random_19], properties {inputVolume [1,0.6]}}
 {id=Mixer_6, children [Attenuation_6, Attenuation_7], properties {inputVolume [1,1]}}
 {id=Mixer_8, children [Random_2,Random_1], properties {inputVolume [1,0.7]}}
 {id=Modulator_0, children [Random_11], properties {pitchMax=1.1, pitchMin=1, volumeMax=0.35, volumeMin=0.35}}
 {id=Modulator_1, children [Random_14], properties {pitchMax=1.1, pitchMin=1, volumeMax=0.35, volumeMin=0.35}}
 {id=Modulator_2, children [Mixer_8], properties {pitchMax=1.1, pitchMin=1, volumeMax=0.35, volumeMin=0.35}}
 {id=Modulator_8, children [Mixer_0], properties {pitchMax=1, pitchMin=0.9, volumeMax=1}}
 {id=QualityLevel_1, children [Mixer_6, Attenuation_1, Attenuation_2, Attenuation_2]}
 {
  id=Random_0
  children [WavePlayer_4, WavePlayer_5, WavePlayer_12, WavePlayer_15, WavePlayer_16, WavePlayer_17]
  properties {weights [1,1,1,1,1,1]}
 }
 {id=Random_1, children [WavePlayer_0,WavePlayer_1,WavePlayer_2,WavePlayer_3,WavePlayer_6,WavePlayer_7], properties {weights [1,1,1,1,1,1]}}
 {id=Random_11, children [Random_7,Mixer_1], properties {bRandomizeWithoutReplacement=false, weights [1,0.4]}}
 {id=Random_14, children [Random_18,Mixer_5], properties {bRandomizeWithoutReplacement=false, weights [1,0.4]}}
 {id=Random_15, children [WavePlayer_30, WavePlayer_31, WavePlayer_32, WavePlayer_33], properties {weights [1,1,1,1]}}
 {id=Random_18, children [WavePlayer_49, WavePlayer_50, WavePlayer_51], properties {weights [1,1,1]}}
 {id=Random_19, children [WavePlayer_56, WavePlayer_58, WavePlayer_59], properties {weights [1,1,1]}}
 {
  id=Random_2
  children [WavePlayer_8, WavePlayer_9, WavePlayer_10, WavePlayer_11, WavePlayer_13, WavePlayer_14]
  properties {weights [1,1,1,1,1,1]}
 }
 {id=Random_7, children [WavePlayer_23, WavePlayer_26, WavePlayer_27, WavePlayer_29], properties {weights [1,1,1,1]}}
 {id=WavePlayer_0, properties {soundWaveAssetPtr=&/Pickaxe_Stone_Viper_Female_Swing_Vox_01}}
 {id=WavePlayer_1, properties {soundWaveAssetPtr=&/Pickaxe_Stone_Viper_Female_Swing_Vox_02}}
 {id=WavePlayer_10, properties {soundWaveAssetPtr=&/Pickaxe_Stone_Viper_Female_Swing_02}}
 {id=WavePlayer_11, properties {soundWaveAssetPtr=&/Pickaxe_Stone_Viper_Female_Swing_04}}
 {id=WavePlayer_12, properties {soundWaveAssetPtr=/Game/Athena/Sounds/Weapons/Fire_Axe_Swing_Distant_02}}
 {id=WavePlayer_13, properties {soundWaveAssetPtr=&/Pickaxe_Stone_Viper_Female_Swing_05}}
 {id=WavePlayer_14, properties {soundWaveAssetPtr=&/Pickaxe_Stone_Viper_Female_Swing_06}}
 {id=WavePlayer_15, properties {soundWaveAssetPtr=/Game/Athena/Sounds/Weapons/Fire_Axe_Swing_Distant_03}}
 {id=WavePlayer_16, properties {soundWaveAssetPtr=/Game/Athena/Sounds/Weapons/Fire_Axe_Swing_Distant_05}}
 {id=WavePlayer_17, properties {soundWaveAssetPtr=/Game/Athena/Sounds/Weapons/Fire_Axe_Swing_Distant_06}}
 {id=WavePlayer_2, properties {soundWaveAssetPtr=&/Pickaxe_Stone_Viper_Female_Swing_Vox_03}}
 {id=WavePlayer_23, properties {soundWaveAssetPtr=&Pickaxe_Stone_Viper_Female_Swing_05}}
 {id=WavePlayer_26, properties {soundWaveAssetPtr=&/Pickaxe_Stone_Viper_Female_Swing_01}}
 {id=WavePlayer_27, properties {soundWaveAssetPtr=&Pickaxe_Stone_Viper_Female_Swing_02}}
 {id=WavePlayer_29, properties {soundWaveAssetPtr=&/Pickaxe_Stone_Viper_Female_Swing_03}}
 {id=WavePlayer_3, properties {soundWaveAssetPtr=&/Pickaxe_Stone_Viper_Female_Swing_Vox_04}}
 {id=WavePlayer_30, properties {soundWaveAssetPtr=&Pickaxe_Stone_Viper_Female_Swing_Vox_04}}
 {id=WavePlayer_31, properties {soundWaveAssetPtr=&Pickaxe_Stone_Viper_Female_Swing_Vox_03}}
 {id=WavePlayer_32, properties {soundWaveAssetPtr=&Pickaxe_Stone_Viper_Female_Swing_Vox_02}}
 {id=WavePlayer_33, properties {soundWaveAssetPtr=&Pickaxe_Stone_Viper_Female_Swing_Vox_01}}
 {id=WavePlayer_4, properties {soundWaveAssetPtr=/Game/Athena/Sounds/Weapons/Fire_Axe_Swing_Distant_04}}
 {id=WavePlayer_49, properties {soundWaveAssetPtr=&Pickaxe_Stone_Viper_Female_Swing_01}}
 {id=WavePlayer_5, properties {soundWaveAssetPtr=/Game/Athena/Sounds/Weapons/Fire_Axe_Swing_Distant_01}}
 {id=WavePlayer_50, properties {soundWaveAssetPtr=&Pickaxe_Stone_Viper_Female_Swing_02}}
 {id=WavePlayer_51, properties {soundWaveAssetPtr=&Pickaxe_Stone_Viper_Female_Swing_03}}
 {id=WavePlayer_56, properties {soundWaveAssetPtr=&Pickaxe_Stone_Viper_Female_Swing_Vox_03}}
 {id=WavePlayer_58, properties {soundWaveAssetPtr=&Pickaxe_Stone_Viper_Female_Swing_Vox_02}}
 {id=WavePlayer_59, properties {soundWaveAssetPtr=&Pickaxe_Stone_Viper_Female_Swing_Vox_01}}
 {id=WavePlayer_6, properties {soundWaveAssetPtr=&/Pickaxe_Stone_Viper_Female_Swing_Vox_05}}
 {id=WavePlayer_7, properties {soundWaveAssetPtr=&/Pickaxe_Stone_Viper_Female_Swing_Vox_06}}
 {id=WavePlayer_8, properties {soundWaveAssetPtr=&Pickaxe_Stone_Viper_Female_Swing_01}}
 {id=WavePlayer_9, properties {soundWaveAssetPtr=&Pickaxe_Stone_Viper_Female_Swing_03}}
]
properties {
 audioPropertiesSheet=/BRCosmetics/Sounds/Settings/PropertySheets/Pickaxe/APS_Cosmetic_Pickaxe_Swing
 bOverrideConcurrency=true
 bPrimeOnLoad=true
 maxDistance=3000
 soundClassObject=&Fort_WeaponMelee_PSM
 volumeMultiplier=0.3
 concurrencyOverrides {bLimitToOwner=true, maxCount=2, resolutionRule=StopOldest}
 concurrencySet [&Player_Melee_Weapon_Conc]
}
